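# Niagara system for the PowerBeam vehicle drift trail cosmetic, trailing ribbons and lightning-like bursts from the wheels.
# A quality-scaled ribbon emitter feeds two renderers, while two mesh/sprite emitters spawn short-lived SubUV-animated bursts at skeletal mesh points shifted by User.offset.
# All three emitters use FN_DisableTireFX for per-wheel masking and painted-variant overrides for colour.
# One burst emitter's colour scaling reads User.IsPainted, whereas the others read User.bIsPainted.
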NiagaraSystem /VehicleCosmetics/DriftTrail/DriftTrail_PowerBeam/FX/NS_DriftTrail_PowerBeam_MP
dependencies {
 usedDataInterfaces [
  &NiagaraDataInterfaceMeshRendererInfo
  &NiagaraDataInterfaceCurve
  &NiagaraDataInterfaceVectorCurve
  &NiagaraDataInterfaceSkeletalMesh
 ]
 usedDynamicInputs []
 usedModules [
  &self:SystemSpawnScript.SystemScriptSource.SystemScriptGraph.NiagaraNodeAssignment_0.SetVariables_#1
  &SystemState
 ]
}
emitters {
 NE_VC_DriftTrail_Ribbon_Parent002 {
  emitterUpdateScript [
   &EmitterState()
   &VC_RateQualityHandler(Epic Spawn Rate=90, High Spawn Rate=60, Low Spawn Rate=30, Medium Spawn Rate=45, Use Custom Rate?=true)
   &VC_SpawnRibbon_v1(Number of Ribbons=2, SpawnRate=$Emitter.Spawn Rate)
  ]
  particleSpawnScript [
   &VC_PaintedVariantsOverride(
    Black=<rgba 0.19 0.19 0.19 4>
    BurntSienna=<rgba 0.55 0.2305 0.0825 4>
    Cobalt=<rgba 0.2 0.4 1 2.5>
    Crimson=<rgba 0.8 0 0.0533 2>
    DefaultColor=<rgba 0.7 0 0.525 2>
    ForestGreen=<rgba 0.4117 0.65 0.2925 2.5>
    Grey=<rgba 0.36 0.36 0.36 2.5>
    Lime=<rgba 0.636 0.8 0.144 2.5>
    Orange=<rgba 1 0.265 0.02 2>
    Pink=<rgba 0.85 0.17 0.7367 2>
    Purple=<rgba 0.4831 0.255 0.85 2>
    R1 Black:LinearColor=<rgba 1 1 1 1>
    R1 BurntSienna:LinearColor=<rgba 2.3 0.6679 0.23 10>
    R1 Cobalt:LinearColor=<rgba 0 0.074 4 1>
    R1 Crimson:LinearColor=<rgba 2 0.2 0.2488 1>
    R1 DefaultColor=<rgba 0 1.5 1.5 1>
    R1 ForestGreen:LinearColor=<rgba 1 2 1.0505 1>
    R1 Grey:LinearColor=<rgba 1 1 1 1>
    R1 Lime:LinearColor=<rgba 1.5833 2 1 1>
    R1 Orange:LinearColor=<rgba 1.5 0.7875 0.075 2>
    R1 Pink:LinearColor=<rgba 2 1 1.6364 1>
    R1 Purple:LinearColor=<rgba 1.4104 1 2 1>
    R1 Saffron:LinearColor=<rgba 2 1.7667 1 1>
    R1 SkyBlue:LinearColor=<rgba 1 1.6654 2 1>
    R1 TitaniumWhite:LinearColor=<rgba 0 0 0 1>
    Saffron=<rgba 0.95 0.7727 0.19 2>
    SkyBlue=<rgba 0 0.8 0.95 2>
    TitaniumWhite=<rgba 0.85 0.85 0.85 2>
    Use R1 Painted Overrides=true
   )
   &VC_InitializeRibbon_v1(
    Color=$StackContext.Color
    Custom Offset=true
    Enable Skeletal Position=true
    Lifetime=0.7
    Local Vector Offset=&AddVectorToPosition(Position=<-45 0 -30>, Vector=$User.offset)
    Offset Amount=1
    Ribbon Width=50
   )
   &AddVelocity(Velocity=<-1400 0 0>) {module=AddVelocity001}
   /VehicleCosmetics/FX/Niagara/Modules/MergeEmitters/VC_ExtraBindingColors(
    ExtraColors=1
    R1 Color:LinearColor=&MakeCustomLinearColorFromBool(Bool=$User.bIsPainted, True Color=$User.PaintedColor, False Color=<rgba 0 2 2 1>)
   ) {enabled=false}
   /SetVariables(Particles.Width Ray=80) {isSetParameters=true}
  ]
  particleUpdateScript [
   &FN_DisableTireFX(Mask Index Attribute:NiagaraInt32=$Particles.SpawnGroup, Mask Index FL=2, Mask Index FR=3, Mask Index RL=0, Mask Index RR=1)
   &InheritSourceMovement(
    Applied Position Scale=&Multiply_VectorByFloat(
     Vector=<1 1 1>
     Float=&FloatFromCurve(FloatCurve=DataInterface() {propertyValues {Curve {keys [{interpMode=RCIM_Cubic, tangentMode=RCTM_Break, time=0, value=1}, {interpMode=RCIM_Cubic, tangentMode=RCTM_Break, time=1, value=0}]}}})
    )
    ApplyForce=false
    ApplyPosition=true
    ApplyPositionFalloff=false
   )
   &ParticleState()
   &SolveForcesAndVelocity(Clamp Velocity=false, Limit Acceleration=false)
   &ScaleColor(
    Scale Alpha=&FloatFromCurve(
     FloatCurve=DataInterface() {propertyValues {Curve {keys [
         {arriveTangent=-1.661, interpMode=RCIM_Cubic, leaveTangent=-1.661, tangentMode=RCTM_User, time=0, value=1}
         {arriveTangent=-1.8893, interpMode=RCIM_Cubic, leaveTangent=-1.8893, tangentMode=RCTM_User, time=0.8, value=0.5}
         <at 1 0>
        ]}}}
     Scale Curve=7
    )
    Scale RGB=&MakeCustomVectorFromBool(
     Bool=$User.bIsPainted
     True Vector=&Multiply_VectorByFloat(
      Vector=<1 1 1>
      Float=&FloatFromCurve(FloatCurve=DataInterface() {propertyValues {Curve {keys [<at 0 1 RCIM_Cubic>,<at 0.7 0.2 RCIM_Cubic>]}}})
     )
     False Vector=&VectorFromCurve(VectorCurve=DataInterface() {propertyValues {
        XCurve {keys [<at 0 1>,<at 0.3 0.4666>,<at 0.6 0.3>,<at 0.9 0>]}
        YCurve {keys [<at 0 0>,<at 0.9 0>]}
        ZCurve {keys [<at 0 2>,<at 0.3 1.1666>,<at 0.6 1>,<at 0.9 0>]}
       }})
    )
    ScaleA=true
    ScaleRGB=true
   )
   &ScaleRibbonWidth(Ribbon Width Scale=&FloatFromCurve(FloatCurve=DataInterface() {propertyValues {Curve {keys [
         {arriveTangent=3.9945, interpMode=RCIM_Cubic, leaveTangent=3.9945, tangentMode=RCTM_User, time=0, value=0.25}
         {
          arriveTangentWeight=0.1403
          interpMode=RCIM_Cubic
          leaveTangentWeight=0.1403
          tangentMode=RCTM_User
          tangentWeightMode=RCTWM_WeightedBoth
          time=0.27
          value=1
         }
         <at 0.8 0.3>
         <at 1 0>
        ]}}}))
   &VC_UpdateRibbon_v1(Enable Skeletal Position=true)
   /SetVariables(Particles.R1 ColorRay:LinearColor=&MakeCustomLinearColorFromBool(Bool=$User.bIsPainted, True Color=$User.PaintedColor, False Color=<rgba 0 2 2 1>)) {enabled=false, isSetParameters=true}
   /SetVariables(
    Particles.R1 ColorB:LinearColor=&VC_DynamicScaleColor(
     Color to Scale:LinearColor=$Particles.R1 ColorRay
     RGB Scale:Vector3f=&VectorFromFloat(Value=$User.EmissiveLightning)
     Alpha Scale:NiagaraFloat=&FloatFromCurve(
      FloatCurve=DataInterface() {propertyValues {Curve {keys [
          {arriveTangent=4.7274, interpMode=RCIM_Cubic, leaveTangent=4.7274, tangentMode=RCTM_User, time=0, value=0}
          {arriveTangent=-1.8122, interpMode=RCIM_Cubic, leaveTangent=-1.8122, tangentMode=RCTM_User, time=0.3534, value=0.55}
          <at 0.5534 0>
         ]}}}
      Scale Curve=0
     )
    )
    Particles.Width Ray G:NiagaraFloat=&Multiply_Float(
     A=$Particles.Width Ray
     B=&FloatFromCurve(FloatCurve=DataInterface() {propertyValues {Curve {keys [<at 0 0>,<at 0.0267 0.425>,<at 0.0715 0.6625>,<at 0.1378 0.9>,<at 0.35 1.2>,<at 0.82 0.8>,<at 0.97 0>]}}})
    )
   ) {isSetParameters=true}
  ]
  rendererClasses [&NiagaraRibbonRendererProperties]
  renderers [
   {rendererClass=&NiagaraRibbonRendererProperties}
   {rendererClass=&NiagaraRibbonRendererProperties, rendererIndex=1}
  ]
 }
 NE_VC_DriftTrail_Sprite_Parent {
  emitterUpdateScript [
   &EmitterState()
   &SpawnRate(SpawnRate=24, Use Spawn Probability=false)
  ]
  particleSpawnScript [
   &VC_PaintedVariantsOverride(
    Black=<rgba 0.04 0.04 0.04 1>
    BurntSienna=<rgba 0.3 0.0635 0 1>
    Cobalt=<rgba 0 0.2 1.2 1>
    Crimson=<rgba 0.55 0 0.0149 1>
    DefaultColor=<rgba 0 0.6667 0.8 1>
    ForestGreen=<rgba 0.1497 0.4 0.1 1>
    Grey=<rgba 0.15 0.15 0.15 1>
    Lime=<rgba 0.3 0.6 0 1>
    Orange=<rgba 0.6 0.07 0 1>
    Pink=<rgba 0.65 0.0779 0.442 1>
    Purple=<rgba 0.3163 0.055 1.1 1>
    Saffron=<rgba 0.45 0.3555 0.045 1>
    SkyBlue=<rgba 0 0.3327 0.5 1>
    TitaniumWhite=<rgba 0.6 0.6 0.6 1>
   )
   &InitializeParticle(
    Color=$StackContext.Color
    Lifetime=&RandomRangeFloat(Minimum=0.2, Maximum=0.4)
    Mesh Scale Max=<2 0.2 2>
    Mesh Scale Min=<2 4 7>
    Mesh Scale Mode=::NewEnumerator2
    UsePositionOffset=false
    Write Material Random=true
   )
   &PartitionParticles()
   &SkeletalMeshLocation(
    Bone / Socket Index=&Add_Integer(A=$Particles.PartitionParticles.Partition, B=4)
    Bone / Socket Sampling Mode=::NewEnumerator5
    Interpolate Mesh Sampling=false
    MaskSpawnGroup=false
    Mesh Sampling Type=::NewEnumerator6
    Sampled Position Offset=&AddVectorToPosition(Position=<-65 0 -30>, Vector=$User.offset)
    Skeletal Mesh=$System.SkeletalMesh
   )
   &AddVelocity(Velocity=&RandomRangeVector(Minimum=<-100 0 0>, Maximum=<-300 1 1>))
   &InitialMeshOrientation(
    Mesh Axis To Orient=&RandomRangeVector(Minimum=<-1 -1 -0.7>, Maximum=<1 1 -1.8>)
    Mesh Orientation Mode=::NewEnumerator2
    Orientation Vector=<0 0 0>
    Rotation=&RandomRangeVector(Minimum=<0.5 0.5 1>, Maximum=<-0.5 -0.5 1>)
    Use Rotation Vector=true
   ) {enabled=false}
   &InitialMeshOrientation(
    Mesh Axis To Orient=&RandomRangeVector(Minimum=<-1 -1 -0.7>, Maximum=<1 1 -1.8>)
    Mesh Orientation Mode=::NewEnumerator2
    Rotation=&RandomRangeVector(Minimum=<-0.5 0.8 0.8>, Maximum=<0.5 1.2 0.7>)
    Use Rotation Vector=true
   ) {module=InitialMeshOrientation001}
   &ShapeLocation(MaskSpawnGroup=false, Override Seed=false, Sphere Radius=18, UseOwnerScale=true)
  ]
  particleUpdateScript [
   &FN_DisableTireFX(
    Mask Index Attribute:NiagaraInt32=$Particles.PartitionParticles.Partition
    Mask Index FL=2
    Mask Index FR=3
    Mask Index RL=0
    Mask Index RR=1
   )
   &SubUVAnimation(
    Mesh Renderer=DataInterface() {propertyValues {MeshRenderer=&self:NE_VC_DriftTrail_Sprite_Parent.NiagaraMeshRendererProperties_0}}
    Renderer Type:ENiagaraMeshOrSprite=::NewEnumerator1
    SubUV Animation Mode:ENiagara_SubUVLookupModeV2=::NewEnumerator1
    SubUV Curve=DataInterface() {propertyValues {Curve {keys [{tangentMode=RCTM_User, time=0, value=0}, {tangentMode=RCTM_User, time=1, value=1}]}}}
    UseEndFrame=false
    UseStartFrame=false
   )
   &ParticleState()
   &ScaleColor(
    Scale RGB=&MakeCustomVectorFromBool(Bool=$User.IsPainted, True Vector=&VectorFromFloat(Value=0.7), False Vector=<2 2 2>)
    ScaleA=true
    ScaleRGB=true
   )
   &SolveForcesAndVelocity(Clamp Velocity=false, Limit Acceleration=false)
   &DynamicMaterialParameters(
    Index 0 Param 2=2
    Index 0 Param 3=&FloatFromCurve(FloatCurve=DataInterface() {propertyValues {Curve {defaultValue=1, keys [{tangentMode=RCTM_User, time=0, value=0}, {tangentMode=RCTM_User, time=1, value=3}]}}})
    Module.Index0Float0Write=true
    Module.Index0Float1Write=true
    Module.Index0Float2Write=true
    Module.Index0Float3Write=true
   )
  ]
  rendererClasses [&NiagaraMeshRendererProperties, &NiagaraSpriteRendererProperties]
  renderers [
   {rendererClass=&NiagaraMeshRendererProperties, rendererIndex=0}
   {rendererClass=&NiagaraMeshRendererProperties, rendererIndex=1}
   {rendererClass=&NiagaraSpriteRendererProperties, rendererIndex=2}
  ]
 }
 NE_VC_DriftTrail_Sprite_Parent001 {
  emitterUpdateScript [
   &EmitterState()
   &SpawnRate(SpawnRate=32, Use Spawn Probability=false)
  ]
  particleSpawnScript [
   &VC_PaintedVariantsOverride(
    Black=<rgba 0.025 0.025 0.025 5>
    BurntSienna=<rgba 0.21 0.0329 0.0021 3>
    Cobalt=<rgba 0.075 0.1529 0.5 3>
    Crimson=<rgba 0.4 0.008 0.0145 3>
    DefaultColor=<rgba 0.4 0.4 0.4 10>
    ForestGreen=<rgba 0.1125 0.3 0.075 3>
    Grey=<rgba 0.1 0.1 0.1 3>
    Lime=<rgba 0.3 0.6 0 3>
    Orange=<rgba 0.6 0.07 0 3>
    Pink=<rgba 0.45 0.0539 0.306 3>
    Purple=<rgba 0.1625 0.05 0.5 3>
    Saffron=<rgba 0.35 0.2765 0.035 3>
    SkyBlue=<rgba 0 0.6654 1 1>
    TitaniumWhite=<rgba 0.4 0.4 0.4 2>
   )
   &InitializeParticle(
    Color=$StackContext.Color
    Lifetime=&RandomRangeFloat(Minimum=0.2, Maximum=0.2)
    Mesh Scale Max=<2.028 4.225 2.028>
    Mesh Scale Min=<1.69 2.028 1.69>
    Mesh Scale Mode=::NewEnumerator2
    UsePositionOffset=false
    Write Material Random=true
   )
   &PartitionParticles()
   &SkeletalMeshLocation(
    Bone / Socket Index=&Add_Integer(A=$Particles.PartitionParticles.Partition, B=4)
    Bone / Socket Sampling Mode=::NewEnumerator5
    Interpolate Mesh Sampling=false
    MaskSpawnGroup=false
    Mesh Sampling Type=::NewEnumerator6
    Sampled Position Offset=&AddVectorToPosition(Position=<-50 0 -30>, Vector=$User.offset)
    Skeletal Mesh=$System.SkeletalMesh
   )
   &AddVelocity(Velocity=&RandomRangeVector(Minimum=<-50 0 0>, Maximum=<-250 1 1>)) {enabled=false}
   &InitialMeshOrientation(
    Mesh Axis To Orient=<0.2 1 0>
    Mesh Orientation Mode=::NewEnumerator2
    Orientation Coordinate Space=::Local
    Orientation Vector=<0 0 0>
    Rotation=&RandomRangeVector(Minimum=<0 -0.3 0>, Maximum=<0 0.3 0>)
    Use Rotation Vector=true
   )
   /SetVariables(Particles.SizeLightningStrike:Vector3f=&RandomRangeVector(Minimum=<5 1 5>, Maximum=<3.5 5.6 3.5>)) {isSetParameters=true}
  ]
  particleUpdateScript [
   &FN_DisableTireFX(
    Mask Index Attribute:NiagaraInt32=$Particles.PartitionParticles.Partition
    Mask Index FL=2
    Mask Index FR=3
    Mask Index RL=0
    Mask Index RR=1
   )
   &ParticleState()
   /SetVariables(Particles.ScaleLightningStrike:Vector3f=&Multiply_Vector(
     A=$Particles.SizeLightningStrike
     B=&VectorFromCurve(VectorCurve=DataInterface() {propertyValues {XCurve {keys [<at 0 0.6>,<at 1 1>]}, YCurve {keys [<at 0 0.4>,<at 1 1>]}, ZCurve {keys [<at 0 0.6>,<at 1 1>]}}})
    )) {isSetParameters=true}
   &ScaleMeshSize(Scale Factor=&VectorFromCurve(VectorCurve=DataInterface() {propertyValues {XCurve {keys [<at 0 0.6>,<at 1 1>]}, YCurve {keys [<at 0 0.4>,<at 1 1>]}, ZCurve {keys [<at 0 0.6>,<at 1 1>]}}}))
   &ScaleColor(
    Scale RGB=&MakeCustomVectorFromBool(
     Bool=$User.bIsPainted
     True Vector=<1 1 1>
     False Vector=&VectorFromCurve(VectorCurve=DataInterface() {propertyValues {XCurve {keys [<at 0.2 0.35>,<at 1.6 0.025>]}, YCurve {keys [<at 0.2 0.05>,<at 1.6 1.025>]}, ZCurve {keys [<at 0 1>,<at 1.6 1.025>]}}})
    )
    ScaleA=true
    ScaleRGB=true
   ) {module=ScaleColor001}
   &SubUVAnimation(
    Mesh Renderer=DataInterface() {propertyValues {MeshRenderer=&self:NE_VC_DriftTrail_Sprite_Parent001_3.NiagaraMeshRendererProperties_0}}
    Renderer Type:ENiagaraMeshOrSprite=::NewEnumerator1
    SubUV Animation Mode:ENiagara_SubUVLookupModeV2=::NewEnumerator1
    SubUV Curve=DataInterface() {propertyValues {Curve {keys [{tangentMode=RCTM_User, time=0, value=0}, {tangentMode=RCTM_User, time=1, value=1}]}}}
    UseEndFrame=false
    UseStartFrame=false
   )
   &SolveForcesAndVelocity(Clamp Velocity=false, Limit Acceleration=false)
  ]
  rendererClasses [&NiagaraMeshRendererProperties, &NiagaraSpriteRendererProperties]
  renderers [
   {rendererClass=&NiagaraMeshRendererProperties, rendererIndex=0}
   {rendererClass=&NiagaraSpriteRendererProperties, rendererIndex=1}
  ]
 }
}
summary {
 systemName=NS_DriftTrail_PowerBeam_MP
 emitters [
  {
   emitterName=NE_VC_DriftTrail_Sprite_Parent
   rendererClasses [&NiagaraMeshRendererProperties, &NiagaraSpriteRendererProperties]
  }
  {
   emitterName=NE_VC_DriftTrail_Sprite_Parent001
   rendererClasses [&NiagaraMeshRendererProperties, &NiagaraSpriteRendererProperties]
  }
  {emitterName=NE_VC_DriftTrail_Ribbon_Parent002, rendererClasses [&NiagaraRibbonRendererProperties]}
 ]
 userVariables [
  User.bIsFrontendPreview {type=NiagaraBool}
  User.bIsPainted {type=NiagaraBool}
  User.ColorVariantID {type=NiagaraInt32}
  User.EmissiveLightning {type=NiagaraFloat, defaultValue=2, description=""}
  User.IsPainted {type=NiagaraBool}
  User.offset {type=Vector3f, description="offset from wheels(set through cosmetic template)", defaultValue {struct=Vector3f, value=<-30 0 0>}}
  User.PaintedColor {type=LinearColor, description="", defaultValue {struct=LinearColor, value=<rgba 1 1 1 1>}}
  User.PaintedScalar {type=NiagaraFloat, defaultValue=1, description=""}
  User.TireOffsetPosition {type=Vector3f, description="", defaultValue {struct=Vector3f, value=<-30 0 0>}}
 ]
}
