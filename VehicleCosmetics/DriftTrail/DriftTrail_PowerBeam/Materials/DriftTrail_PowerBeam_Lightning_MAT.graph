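# Unlit translucent material for Niagara ribbons that draws the lightning of the PowerBeam vehicle drift trail.
# A lightning texture pans along U and is tinted by particle color, with the product of their alphas computed separately.
# A U-based falloff, shaped by the dynamic Power parameter plus a constant offset and scaled by dynamic Brightness, boosts that color.
# The result passes through the shared MF_FX_EmissiveAdjustments function.
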Material /VehicleCosmetics/DriftTrail/DriftTrail_PowerBeam/Materials/DriftTrail_PowerBeam_Lightning_MAT
graph {
 properties {
  bAutomaticallySetUsageInEditor=false
  bUsedWithNiagaraRibbons=true
  blendMode=BLEND_Translucent
  pixelDepthOffsetMode=PDOM_Legacy
  shadingModel=MSM_Unlit
  usedShadingModels=Unlit
 }
 nodes [
  n1 = ComponentMask(None=TextureCoordinate()) {r=true}
  n2 = DynamicParameter() {paramNames [Brightness,Power,Param3,Param4]}
  n3 = Add(A=Power(Base=OneMinus(None=@n1), Exp=@n2.Power))
  MaterialExpressionVertexColor_1 = ParticleColor()
  n4 = TextureSample(UVs=Panner(Coordinate=TextureCoordinate()) {speedX=-2}) {texture=/DealershipSharedLibrary/FX/Textures/Gradient_Masks/T_VCVFX_PowerBeam_Lightning_A}
  MaterialFunctionCall(Emissive=Multiply(A=Multiply(A=@n3, B=@n2.Brightness), B=Multiply(A=@MaterialExpressionVertexColor_1.RGB, B=@n4.RGB))) {materialFunction=&MF_FX_EmissiveAdjustments}
  Multiply(A=@MaterialExpressionVertexColor_1.A, B=@n4.A)
 ]
}
