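# Unlit, two-sided translucent Niagara ribbon material for the PowerBeam drift trail vehicle cosmetic.
# A panning beam texture, offset by a dynamic parameter, is tinted by particle color and passed through MF_FX_EmissiveAdjustments.
# The Desaturate switch grayscales the texture before tinting and applies a U-gradient mask shaped by dynamic parameters.
# Alpha combines particle and texture alpha with U-axis falloffs, broken up by two noise layers scrolling at MaskSpeed.
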Material /VehicleCosmetics/DriftTrail/DriftTrail_PowerBeam/Materials/DriftTrail_PowerBeam_Ribbon_Mat
graph {
 properties {
  bAutomaticallySetUsageInEditor=false
  bUsedWithNiagaraRibbons=true
  blendMode=BLEND_Translucent
  pixelDepthOffsetMode=PDOM_Legacy
  shadingModel=MSM_Unlit
  twoSided=true
  usedShadingModels=Unlit
 }
 nodes [
  n1 = ComponentMask(None=TextureCoordinate()) {r=true}
  n2 = DynamicParameter() {paramNames [Offest,Lightmask,Mask,Param4]}
  n3 = Multiply(A=Power(Base=OneMinus(None=@n1), Exp=@n2.Mask), B=@n2.Lightmask)
  MaterialExpressionVertexColor_0 = ParticleColor()
  n4 = TextureCoordinate() {uTiling=3}
  n5 = AppendVector(A=@n2.Offest, B=Constant())
  n6 = TextureSample(UVs=Panner(Coordinate=Add(A=@n4, B=@n5)) {speedX=-2.65}) {texture=/DealershipSharedLibrary/FX/Textures/Diffuse/T_VCVFX__PowerBeam_Ribbon_D}
  n7 = Multiply(A=@MaterialExpressionVertexColor_0.RGB, B=Desaturation(None=@n6.RGB, Fraction=Constant() {r=1}))
  MaterialFunctionCall(Emissive=StaticSwitchParameter(True=Multiply(A=@n3, B=@n7), False=Multiply(A=@MaterialExpressionVertexColor_0.RGB, B=@n6.RGB)) {parameterName=Desaturate}) {materialFunction=&MF_FX_EmissiveAdjustments}
  n8 = TextureCoordinate()
  n9 = Power(Base=OneMinus(None=ComponentMask(None=@n8) {r=true}), Exp=Constant() {r=6})
  n10 = Multiply(A=ScalarParameter() {defaultValue=1, parameterName=MaskSpeed}, B=Time())
  n11 = TextureSample(UVs=Panner(Coordinate=TextureCoordinate() {uTiling=0.25, vTiling=2}, Time=@n10) {speedX=-1, speedY=-0.1}) {texture=/VehicleCosmetics/FX/Textures/Noise/T_VCVFX_NoiseStarBeam_A}
  n12 = Constant() {r=1.5}
  n13 = TextureSample(UVs=Panner(Coordinate=TextureCoordinate() {uTiling=2}, Time=@n10) {speedX=-1, speedY=-0.1}) {texture=&T_VCVFX_NoiseStarBeam_A}
  n14 = Constant() {r=1.5}
  MaterialExpressionConstantClamp_2 = Saturate(None=Add(A=Multiply(A=@n11.RGB, B=@n12), B=Multiply(A=@n13.RGB, B=@n14)))
  MaterialExpressionConstantClamp_1 = Saturate(None=Add(A=@n9, B=Multiply(A=@n9, B=@MaterialExpressionConstantClamp_2)))
  n15 = Power(Base=ComponentMask(None=TextureCoordinate()) {r=true}, Exp=Constant() {r=1.8})
  n16 = Constant() {r=500}
  MaterialExpressionConstantClamp_0 = Saturate(None=Multiply(A=@MaterialExpressionConstantClamp_1, B=Multiply(A=@n15, B=@n16)))
  Multiply(A=@MaterialExpressionVertexColor_0.A, B=Multiply(A=@n6.A, B=@MaterialExpressionConstantClamp_0))
 ]
}
